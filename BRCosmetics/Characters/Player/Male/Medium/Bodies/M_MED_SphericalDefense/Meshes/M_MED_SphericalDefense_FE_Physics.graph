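# Physics asset for the male medium SphericalDefense player body (FE variant), giving capsule collision to the lower body and its dangling armor bones.
# Capsules on pelvis, spine_01, thighs, calves and feet cover the body, alongside dyn_ bones for front, back and leg armor.
# Front and back armor each form a five-bone chain from a master bone, joined by constraints with tight swing and twist limits.
# Leg armor "unbend" bodies and core body capsules have no constraints.
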PhysicsAsset /BRCosmetics/Characters/Player/Male/Medium/Bodies/M_MED_SphericalDefense/Meshes/M_MED_SphericalDefense_FE_Physics
bodies {
 calf_l [{
   shapeType=Capsule
   length=22.545
   radius=7.455
   shapeName=calf_l_capsule
   center=<-19.8825 -2.6363 0.3261>
   rotation=<rot -89.0604 -179.9983 -175.8154>
  }]
 calf_r [{
   shapeType=Capsule
   length=22.545
   radius=7.455
   shapeName=calf_r_capsule
   center=<19.3765 3.33 -0.3178>
   rotation=<rot -89.0604 -179.9983 -175.8154>
  }]
 dyn_backArmor_01 [{shapeType=Capsule, length=16.6852, radius=1.1161, shapeName=dyn_backArmor_01_capsule, center=<0.5682 0 0>}]
 dyn_backArmor_02 [{shapeType=Capsule, length=13.0988, radius=3.7, shapeName=dyn_backArmor_02_capsule, center=<4.2596 1.0163 0>}]
 dyn_backArmor_03 [{shapeType=Capsule, length=13.7225, radius=4.0787, shapeName=dyn_backArmor_03_capsule, center=<4.6657 1.1142 0>}]
 dyn_backArmor_04 [{shapeType=Capsule, length=13.2391, radius=4.5622, shapeName=dyn_backArmor_04_capsule, center=<5.0124 0.8809 0>}]
 dyn_backArmor_master [{shapeType=Capsule, length=11.9377, radius=0.9349, shapeName=dyn_backArmor_master_capsule, center=<-1.9482 -1.3175 0>}]
 dyn_frontArmor_01 [{shapeType=Capsule, length=11.9685, radius=1.1368, shapeName=dyn_frontArmor_01_capsule, center=<1.3075 0.6008 0>}]
 dyn_frontArmor_02 [{shapeType=Capsule, length=11.4142, radius=3.593, shapeName=dyn_frontArmor_02_capsule, center=<2.5448 1.9904 0>}]
 dyn_frontArmor_03 [{shapeType=Capsule, length=14.152, radius=3.593, shapeName=dyn_frontArmor_03_capsule, center=<4.7867 0.983 0>}]
 dyn_frontArmor_04 [{shapeType=Capsule, length=15, radius=4, shapeName=dyn_frontArmor_04_capsule, center=<5.1595 0.2407 0>}]
 dyn_frontArmor_master [{shapeType=Capsule, length=1, radius=1, shapeName=dyn_frontArmor_master_capsule}]
 dyn_legArmor_l_01_unbendBack_02 [{
   shapeType=Capsule
   length=8.9726
   radius=2.1904
   shapeName=dyn_legArmor_l_01_unbendBack_02_capsule
   center=<5.3835 -4.4329 -2.8485>
   rotation=<rot -82.786 169.9258 -178.4869>
  }]
 dyn_legArmor_l_01_unbendFront_02 [{
   shapeType=Capsule
   length=8.9726
   radius=2.1904
   shapeName=dyn_legArmor_l_01_unbendFront_02_capsule
   center=<5.0727 3.1799 -2.6397>
   rotation=<rot -83.5201 -169.9392 -179.5388>
  }]
 dyn_legArmor_l_02_unbendBack_01 [{
   shapeType=Capsule
   length=8.9726
   radius=2.1904
   shapeName=dyn_legArmor_l_02_unbendBack_01_capsule
   center=<5.7381 -11.4314 -4.5045>
   rotation=<rot -87.6559 177.1793 175.2919>
  }]
 dyn_legArmor_l_03_unbendBack_01 [{
   shapeType=Capsule
   length=8.9726
   radius=2.1904
   shapeName=dyn_legArmor_l_03_unbendBack_01_capsule
   center=<5.8811 -12.5033 -5.2772>
   rotation=<rot -89.7156 179.9909 169.2688>
  }]
 dyn_legArmor_r_01_unbendBack_02 [{
   shapeType=Capsule
   length=8.9726
   radius=2.1904
   shapeName=dyn_legArmor_r_01_unbendBack_02_capsule
   center=<-5.3263 4.4088 2.4083>
   rotation=<rot -84.3806 171.0977 179.5277>
  }]
 dyn_legArmor_r_01_unbendFront_02 [{
   shapeType=Capsule
   length=8.9726
   radius=2.1904
   shapeName=dyn_legArmor_r_01_unbendFront_02_capsule
   center=<-4.9515 -2.7559 2.6065>
   rotation=<rot -85.2784 -171.3562 -179.4472>
  }]
 dyn_legArmor_r_02_unbendBack_01 [{
   shapeType=Capsule
   length=8.9726
   radius=2.1904
   shapeName=dyn_legArmor_r_02_unbendBack_01_capsule
   center=<-5.4505 10.7092 5.2568>
   rotation=<rot -87.5651 171.5778 178.8942>
  }]
 dyn_legArmor_r_03_unbendBack_01 [{
   shapeType=Capsule
   length=8.9726
   radius=2.1904
   shapeName=dyn_legArmor_r_03_unbendBack_01_capsule
   center=<-5.4638 12.3256 5.5537>
   rotation=<rot -86.88 -6.2804 0.522>
  }]
 foot_l [{shapeType=Capsule, length=15, radius=9, shapeName=foot_l_capsule, center=<-3 8 -1>, rotation=<rot 0 0 90>}]
 foot_r [{shapeType=Capsule, length=15, radius=9, shapeName=foot_r_capsule, center=<3 -8 1>, rotation=<rot 0 0 90>}]
 pelvis [{shapeType=Capsule, length=8.3135, radius=10.1757, shapeName=pelvis_capsule, center=<-3.4789 0.2957 0>}]
 spine_01 [{shapeType=Capsule, length=7.7633, radius=8.8967, shapeName=spine_01_capsule}]
 thigh_l [{shapeType=Capsule, length=32, radius=8.7, shapeName=thigh_l_capsule, center=<-19.1459 0.0139 0.5628>, rotation=<rot 89.854 179.946 179.9879>}]
 thigh_r [{
   shapeType=Capsule
   length=32
   radius=8.7
   shapeName=thigh_r_capsule
   center=<19.1459 -0.0139 -0.5628>
   rotation=<rot 89.854 179.9472 179.5034>
  }]
}
constraints [
 {bone1Name=dyn_frontArmor_01, bone2Name=dyn_frontArmor_master, swing1LimitDegrees=15, swing2LimitDegrees=2, twistLimitDegrees=1}
 {bone1Name=dyn_backArmor_04, bone2Name=dyn_backArmor_03, swing1LimitDegrees=5, swing2LimitDegrees=5, twistLimitDegrees=1}
 {bone1Name=dyn_backArmor_03, bone2Name=dyn_backArmor_02, swing1LimitDegrees=5, swing2LimitDegrees=5, twistLimitDegrees=1}
 {bone1Name=dyn_backArmor_02, bone2Name=dyn_backArmor_01, swing1LimitDegrees=5, swing2LimitDegrees=2.5, twistLimitDegrees=1}
 {bone1Name=dyn_backArmor_01, bone2Name=dyn_backArmor_master, swing1LimitDegrees=5, swing2LimitDegrees=2.5, twistLimitDegrees=1}
 {bone1Name=dyn_frontArmor_04, bone2Name=dyn_frontArmor_03, swing1LimitDegrees=6, swing2LimitDegrees=6, twistLimitDegrees=3}
 {bone1Name=dyn_frontArmor_03, bone2Name=dyn_frontArmor_02, swing1LimitDegrees=6, swing2LimitDegrees=6, twistLimitDegrees=3}
 {bone1Name=dyn_frontArmor_02, bone2Name=dyn_frontArmor_01, swing1LimitDegrees=15, swing2LimitDegrees=3, twistLimitDegrees=3}
]
