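# Physics asset for the SphericalDefense medium male player body, setting up physics for the dynamic side leg armor.
# Each side chains three multi-capsule armor bodies from a zero-size master anchor, with looser limits at the top joint and very tight ones below.
# Unconstrained capsules cover pelvis, lower spine, thighs, calves and only the right foot.
# Only the left side adds unbend front and back bodies, also unconstrained.
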PhysicsAsset /BRCosmetics/Characters/Player/Male/Medium/Bodies/M_MED_SphericalDefense/Meshes/M_MED_SphericalDefense_Physics_Sides
bodies {
 calf_l [{shapeType=Capsule, length=25, radius=7, shapeName=calf_l_capsule, center=<-20 -2 0>, rotation=<rot 0 90 85>}]
 calf_r [{shapeType=Capsule, length=25, radius=7, shapeName=calf_r_capsule, center=<20 2 0>, rotation=<rot 0 90 85>}]
 dyn_legArmor_l_01 [
  {
   shapeType=Capsule
   length=10.1801
   radius=2.3502
   shapeName=dyn_legArmor_l_01_capsule
   center=<5.8419 0 0.9536>
   rotation=<rot 90 -90 -90.0004>
  }
  {
   shapeType=Capsule
   length=10.1801
   radius=2.3502
   shapeName=dyn_legArmor_l_01_capsule
   center=<4.8664 -10.5628 -3.5232>
   rotation=<rot 86.2538 39.6246 50.0931>
  }
  {
   shapeType=Capsule
   length=10.46
   radius=2.3502
   shapeName=dyn_legArmor_l_01_capsule
   center=<6.4824 7.8146 -4.3849>
   rotation=<rot 39.7027 95.8809 83.1787>
  }
 ]
 dyn_legArmor_l_01_unbendBack_01 [{
   shapeType=Capsule
   length=10.464
   radius=2.3502
   shapeName=dyn_legArmor_l_01_unbendBack_01_capsule
   center=<5.3386 -9.2197 -3.4571>
   rotation=<rot 89.3933 -0.0068 8.7629>
  }]
 dyn_legArmor_l_01_unbendFront_01 [{
   shapeType=Capsule
   length=10.464
   radius=2.3502
   shapeName=dyn_legArmor_l_01_unbendFront_01_capsule
   center=<6.4824 7.8146 -4.3849>
   rotation=<rot 39.7027 95.8809 83.1787>
  }]
 dyn_legArmor_l_02 [
  {
   shapeType=Capsule
   length=11.6428
   radius=2.7959
   shapeName=dyn_legArmor_l_02_capsule
   center=<6.7066 0 2.1385>
   rotation=<rot -90 180 -179.9999>
  }
  {
   shapeType=Capsule
   length=8.4626
   radius=3.3995
   shapeName=dyn_legArmor_l_02_capsule
   center=<6.8074 10.2238 -4.7516>
   rotation=<rot -85.158 -170.0003 -179.9992>
  }
  {
   shapeType=Capsule
   length=8.0285
   radius=2.7959
   shapeName=dyn_legArmor_l_02_capsule
   center=<4.54 -10.6156 -3.5417>
   rotation=<rot -87.3875 170 -179.9997>
  }
 ]
 dyn_legArmor_l_03 [
  {shapeType=Capsule, length=8, radius=4, shapeName=dyn_legArmor_l_03_capsule, center=<7 -10 -1.7>, rotation=<rot -4 -7 -120>}
  {shapeType=Capsule, length=8, radius=4, shapeName=dyn_legArmor_l_03_capsule, center=<8 7 -1>, rotation=<rot 0 0 120>}
 ]
 dyn_legArmor_l_master [{shapeType=Capsule, length=0, radius=0, shapeName=dyn_legArmor_l_master_capsule}]
 dyn_legArmor_r_01 [
  {
   shapeType=Capsule
   length=10.7938
   radius=2.3502
   shapeName=dyn_legArmor_l_01_capsule
   center=<-6.9413 10.5697 3.5794>
   rotation=<rot 82.9906 -7.8836 -179.1802>
  }
  {
   shapeType=Capsule
   length=9.5815
   radius=2.3502
   shapeName=dyn_legArmor_l_01_capsule
   center=<-6.1165 -8.2875 4.0599>
   rotation=<rot 39.4242 94.6995 -97.6166>
  }
  {
   shapeType=Capsule
   length=11.1164
   radius=2.3502
   shapeName=dyn_legArmor_r_01_capsule
   center=<-6.7145 0 -1.1087>
   rotation=<rot -81.9453 -66.9613 66.7563>
  }
 ]
 dyn_legArmor_r_02 [
  {
   shapeType=Capsule
   length=8.3994
   radius=2.4664
   shapeName=dyn_legArmor_r_02_capsule
   center=<-5.8989 0 -0.1252>
   rotation=<rot -88.7841 0.0012 -0.0011>
  }
  {
   shapeType=Capsule
   length=8.3994
   radius=2.4664
   shapeName=dyn_legArmor_r_02_capsule
   center=<-6.2305 11.3888 4.7408>
   rotation=<rot -53.4577 85.2857 -95.644>
  }
  {
   shapeType=Capsule
   length=8.3994
   radius=2.4664
   shapeName=dyn_legArmor_r_02_capsule
   center=<-5.0177 -9.6895 4.9075>
   rotation=<rot -55.526 -87.8185 96.6288>
  }
 ]
 dyn_legArmor_r_03 [
  {shapeType=Capsule, length=8, radius=4, shapeName=dyn_legArmor_r_03_capsule, center=<-7 9 1.2>, rotation=<rot 0 0 -120>}
  {shapeType=Capsule, length=8, radius=4, shapeName=dyn_legArmor_r_03_capsule, center=<-7 -8 2>, rotation=<rot 0 0 120>}
 ]
 dyn_legArmor_r_master [{shapeType=Capsule, length=0, radius=0, shapeName=dyn_legArmor_r_master_capsule}]
 foot_r [{shapeType=Capsule, length=15, radius=8, shapeName=foot_r_capsule, center=<4 -6.5 0>, rotation=<rot 0 0 -90>}]
 pelvis [{shapeType=Capsule, length=8.3135, radius=10.1757, shapeName=pelvis_capsule, center=<-3.4789 0.2957 0>}]
 spine_01 [{shapeType=Capsule, length=7.7633, radius=8.8967, shapeName=spine_01_capsule}]
 thigh_l [{shapeType=Capsule, length=24, radius=8.5, shapeName=thigh_l_capsule, center=<-23 0 0.8>, rotation=<rot 0 90 -90>}]
 thigh_r [{shapeType=Capsule, length=24, radius=8.5, shapeName=thigh_r_capsule, center=<23 0 -0.8>, rotation=<rot 0 90 -90>}]
}
constraints [
 {bone1Name=dyn_legArmor_l_01, bone2Name=dyn_legArmor_l_master, swing1LimitDegrees=25, swing2LimitDegrees=25, twistLimitDegrees=10}
 {bone1Name=dyn_legArmor_l_02, bone2Name=dyn_legArmor_l_01, swing1LimitDegrees=3, swing2LimitDegrees=3, twistLimitDegrees=1}
 {bone1Name=dyn_legArmor_l_03, bone2Name=dyn_legArmor_l_02, swing1LimitDegrees=3, swing2LimitDegrees=3, twistLimitDegrees=1}
 {bone1Name=dyn_legArmor_r_01, bone2Name=dyn_legArmor_r_master, swing1LimitDegrees=25, swing2LimitDegrees=18, twistLimitDegrees=10}
 {bone1Name=dyn_legArmor_r_02, bone2Name=dyn_legArmor_r_01, swing1LimitDegrees=3, swing2LimitDegrees=3, twistLimitDegrees=1}
 {bone1Name=dyn_legArmor_r_03, bone2Name=dyn_legArmor_r_02, swing1LimitDegrees=2, swing2LimitDegrees=3, twistLimitDegrees=1}
]
