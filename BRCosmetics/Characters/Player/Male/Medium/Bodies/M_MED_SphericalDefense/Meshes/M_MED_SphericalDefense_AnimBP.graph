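# Post-process style AnimBlueprint for the SphericalDefense male body cosmetic, adding secondary motion to pouches, armor, tassels and straps.
# It copies the parent mesh pose, then chains weighted bone constraints, a Control Rig, additive armor offsets, AnimDynamics and three RigidBody simulations.
# The update event derives speed, movement-state alphas and gravity override through a deprecated helper.
# IsPistol is only ever set true, including before the weapon switch.
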AnimBlueprint /BRCosmetics/Characters/Player/Male/Medium/Bodies/M_MED_SphericalDefense/Meshes/M_MED_SphericalDefense_AnimBP
target_skeleton=&/M_MED_SphericalDefense_Skeleton
graphs [
 AnimGraph {nodes [
   n1 = "Constraint - dyn_bladeAim_sckt (PPPPP)"!(
    ComponentPose=Local To Component(LocalPose=Copy Pose From Mesh() {bUseAttachedParent=true})
    Alpha=1
    ConstraintWeights_0=0.0184
    ConstraintWeights_1:float=0.2347
    ConstraintWeights_2:float=0.1867
    ConstraintWeights_3:float=0.5434
    ConstraintWeights_4:float=0.0168
   ) -> (Pose) {
    alphaBoolBlend {blendOption=Linear}
    boneToModify {boneName=dyn_bladeAim_sckt}
    constraintSetup [
     {offsetOption=None, targetBone {boneName=upperarm_l}}
     {targetBone {boneName=lowerarm_l}}
     {targetBone {boneName=lowerarm_twist_01_l}}
     {targetBone {boneName=lowerarm_twist_02_l}}
     {targetBone {boneName=deform_elbow_out_l}}
    ]
    constraintWeights [1,1,1,1,1]
   }
   n2 = "Constraint - dyn_chestPouch_a_master (PPPPPPPPP)"!(
    ComponentPose=@n1
    Alpha=1
    ConstraintWeights_0=0.0677
    ConstraintWeights_1:float=0.1269
    ConstraintWeights_2:float=0.2669
    ConstraintWeights_3:float=0.2955
    ConstraintWeights_4:float=0.1118
    ConstraintWeights_5:float=0.0845
    ConstraintWeights_6:float=0.0284
    ConstraintWeights_7:float=0.0172
    ConstraintWeights_8:float=0.0011
   ) -> (Pose) {
    alphaBoolBlend {blendOption=Linear}
    boneToModify {boneName=dyn_chestPouch_a_master}
    constraintSetup [
     {targetBone {boneName=pelvis}}
     {targetBone {boneName=spine_01}}
     {targetBone {boneName=spine_02}}
     {targetBone {boneName=spine_03}}
     {targetBone {boneName=spine_04}}
     {targetBone {boneName=clavicle_l}}
     {targetBone {boneName=upperarm_twist_01_l}}
     {targetBone {boneName=deform_upperarm_bk_l}}
     {targetBone {boneName=deform_pec_l}}
    ]
    constraintWeights [1,1,1,1,1,1,1,1,1]
   }
   n3 = "Constraint - dyn_chestPouch_b_master (PPPPPPPP)"!(
    ComponentPose=@n2
    Alpha=1
    ConstraintWeights_0=0.0631
    ConstraintWeights_1:float=0.1314
    ConstraintWeights_2:float=0.2679
    ConstraintWeights_3:float=0.3217
    ConstraintWeights_4:float=0.1072
    ConstraintWeights_5:float=0.0711
    ConstraintWeights_6:float=0.0101
    ConstraintWeights_7:float=0.0276
   ) -> (Pose) {
    alphaBoolBlend {blendOption=Linear}
    boneToModify {boneName=dyn_chestPouch_b_master}
    constraintSetup [
     {targetBone {boneName=pelvis}}
     {targetBone {boneName=spine_01}}
     {targetBone {boneName=spine_02}}
     {targetBone {boneName=spine_03}}
     {targetBone {boneName=spine_04}}
     {targetBone {boneName=clavicle_l}}
     {targetBone {boneName=upperarm_twist_01_l}}
     {targetBone {boneName=deform_pec_l}}
    ]
    constraintWeights [1,1,1,1,1,1,1,1]
   }
   n4 = "Constraint - dyn_chestPouch_c_master (PPPPPPP)"!(
    ComponentPose=@n3
    Alpha=1
    ConstraintWeights_0=0.0503
    ConstraintWeights_1:float=0.1126
    ConstraintWeights_2:float=0.2746
    ConstraintWeights_3:float=0.4052
    ConstraintWeights_4:float=0.1192
    ConstraintWeights_5:float=0.0316
    ConstraintWeights_6:float=0.0065
   ) -> (Pose) {
    alphaBoolBlend {blendOption=Linear}
    boneToModify {boneName=dyn_chestPouch_c_master}
    constraintSetup [
     {targetBone {boneName=pelvis}}
     {targetBone {boneName=spine_01}}
     {targetBone {boneName=spine_02}}
     {targetBone {boneName=spine_03}}
     {targetBone {boneName=spine_04}}
     {targetBone {boneName=clavicle_l}}
     {targetBone {boneName=deform_pec_l}}
    ]
    constraintWeights [1,1,1,1,1,1,1]
   }
   n5 = "Constraint - dyn_chestPouch_d_master (PPPPPP)"!(
    ComponentPose=@n4
    Alpha=1
    ConstraintWeights_0=0.0569
    ConstraintWeights_1:float=0.1126
    ConstraintWeights_2:float=0.2855
    ConstraintWeights_3:float=0.4123
    ConstraintWeights_4:float=0.1235
    ConstraintWeights_5:float=0.0092
   ) -> (Pose) {
    alphaBoolBlend {blendOption=Linear}
    boneToModify {boneName=dyn_chestPouch_d_master}
    constraintSetup [
     {targetBone {boneName=pelvis}}
     {targetBone {boneName=spine_01}}
     {targetBone {boneName=spine_02}}
     {targetBone {boneName=spine_03}}
     {targetBone {boneName=spine_04}}
     {targetBone {boneName=clavicle_r}}
    ]
    constraintWeights [1,1,1,1,1,1]
   }
   n6 = "Constraint - dyn_chestPouch_e_master (PPPPPPP)"!(
    ComponentPose=@n5
    Alpha=1
    ConstraintWeights_0=0.0488
    ConstraintWeights_1:float=0.1157
    ConstraintWeights_2:float=0.2666
    ConstraintWeights_3:float=0.3868
    ConstraintWeights_4:float=0.1172
    ConstraintWeights_5:float=0.0476
    ConstraintWeights_6:float=0.0175
   ) -> (Pose) {
    alphaBoolBlend {blendOption=Linear}
    boneToModify {boneName=dyn_chestPouch_e_master}
    constraintSetup [
     {targetBone {boneName=pelvis}}
     {targetBone {boneName=spine_01}}
     {targetBone {boneName=spine_02}}
     {targetBone {boneName=spine_03}}
     {targetBone {boneName=spine_04}}
     {targetBone {boneName=clavicle_r}}
     {targetBone {boneName=deform_pec_r}}
    ]
    constraintWeights [1,1,1,1,1,1,1]
   }
   n7 = "Constraint - dyn_chestPouch_f_master (PPPPPPPP)"!(
    ComponentPose=@n6
    Alpha=1
    ConstraintWeights_0=0.0617
    ConstraintWeights_1:float=0.1307
    ConstraintWeights_2:float=0.267
    ConstraintWeights_3:float=0.3258
    ConstraintWeights_4:float=0.108
    ConstraintWeights_5:float=0.0701
    ConstraintWeights_6:float=0.0077
    ConstraintWeights_7:float=0.0292
   ) -> (Pose) {
    alphaBoolBlend {blendOption=Linear}
    boneToModify {boneName=dyn_chestPouch_f_master}
    constraintSetup [
     {targetBone {boneName=pelvis}}
     {targetBone {boneName=spine_01}}
     {targetBone {boneName=spine_02}}
     {targetBone {boneName=spine_03}}
     {targetBone {boneName=spine_04}}
     {targetBone {boneName=clavicle_r}}
     {targetBone {boneName=upperarm_twist_01_r}}
     {targetBone {boneName=deform_pec_r}}
    ]
    constraintWeights [1,1,1,1,1,1,1,1]
   }
   n8 = "Constraint - dyn_chestPouch_g_master (PPPPPPPPP)"!(
    ComponentPose=@n7
    Alpha=1
    ConstraintWeights_0=0.0693
    ConstraintWeights_1:float=0.1297
    ConstraintWeights_2:float=0.2696
    ConstraintWeights_3:float=0.2966
    ConstraintWeights_4:float=0.1092
    ConstraintWeights_5:float=0.0817
    ConstraintWeights_6:float=0.027
    ConstraintWeights_7:float=0.0134
    ConstraintWeights_8:float=0.0035
   ) -> (Pose) {
    alphaBoolBlend {blendOption=Linear}
    boneToModify {boneName=dyn_chestPouch_g_master}
    constraintSetup [
     {targetBone {boneName=pelvis}}
     {targetBone {boneName=spine_01}}
     {targetBone {boneName=spine_02}}
     {targetBone {boneName=spine_03}}
     {targetBone {boneName=spine_04}}
     {targetBone {boneName=clavicle_r}}
     {targetBone {boneName=upperarm_twist_01_r}}
     {targetBone {boneName=deform_upperarm_bk_r}}
     {targetBone {boneName=deform_pec_r}}
    ]
    constraintWeights [1,1,1,1,1,1,1,1,1]
   }
   n9 = Control Rig!(
    Source=Component To Local(ComponentPose=@n8)
    Alpha
    PawnSpeed:double=Get PawnSpeed() -> (PawnSpeed:double)
    playingEmote:bool=Get bIsPlayingEmote()
    Is Pistol Alpha:double=Get Is Pistol Alpha() -> (Is Pistol Alpha:double)
    Is Playing Emote Alpha:double=Get Is Playing Emote Alpha() -> (Is Playing Emote Alpha:double)
   ) -> (Pose) {
    alphaBoolBlend {blendOption=Linear}
    controlRigAssetReference {blueprintRigClass=&/M_MED_SphericalDefense_CtrlRig.M_MED_SphericalDefense_CtrlRig_C}
   }
   n10 = "Transform (Modify) Bone - Bone: dyn_legArmor_r_master"(ComponentPose=Local To Component(LocalPose=@n9), Alpha=1, Translation="0,0,0", Rotation="30,0,0", Scale="1,1,1") {
    rotationMode=BMM_Additive
    rotationSpace=BCS_BoneSpace
    translationMode=BMM_Additive
    translationSpace=BCS_BoneSpace
    alphaBoolBlend {blendOption=Linear}
    boneToModify {boneName=dyn_legArmor_r_master}
    rotation {pitch=30}
   }
   n11 = "Transform (Modify) Bone - Bone: dyn_legArmor_l_master"(ComponentPose=@n10, Alpha=1, Translation="0,0,0", Rotation="30,0,0", Scale="1,1,1") {
    rotationMode=BMM_Additive
    rotationSpace=BCS_BoneSpace
    translationMode=BMM_Additive
    translationSpace=BCS_BoneSpace
    alphaBoolBlend {blendOption=Linear}
    boneToModify {boneName=dyn_legArmor_l_master}
    rotation {pitch=30}
   }
   n12 = "Transform (Modify) Bone - Bone: dyn_frontArmor_master"(ComponentPose=@n11, Alpha=1, Translation="0,0,0", Rotation="0,20,0", Scale="1,1,1") {
    rotationMode=BMM_Additive
    rotationSpace=BCS_BoneSpace
    translationMode=BMM_Additive
    translationSpace=BCS_BoneSpace
    alphaBoolBlend {blendOption=Linear}
    boneToModify {boneName=dyn_frontArmor_master}
    rotation {yaw=20}
   }
   n13 = "Transform (Modify) Bone - Bone: dyn_backArmor_master"(ComponentPose=@n12, Alpha=1, Translation="0,0,0", Rotation="0,40,0", Scale="1,1,1") {
    rotationMode=BMM_Additive
    rotationSpace=BCS_BoneSpace
    translationMode=BMM_Additive
    translationSpace=BCS_BoneSpace
    alphaBoolBlend {blendOption=Linear}
    boneToModify {boneName=dyn_backArmor_master}
    rotation {yaw=40}
   }
   n14 = "Transform (Modify) Bone - Bone: dyn_legArmor_r_master"(ComponentPose=@n13, Alpha=1, Translation="0,0,0", Rotation="60,0,0", Scale="1,1,1") {
    rotationMode=BMM_Additive
    rotationSpace=BCS_BoneSpace
    translationMode=BMM_Additive
    translationSpace=BCS_BoneSpace
    alphaBoolBlend {blendOption=Linear}
    boneToModify {boneName=dyn_legArmor_r_master}
    rotation {pitch=60}
   }
   n15 = "Transform (Modify) Bone - Bone: dyn_legArmor_l_master"(ComponentPose=@n14, Alpha=1, Translation="0,0,0", Rotation="10,0,0", Scale="1,1,1") {
    rotationMode=BMM_Additive
    rotationSpace=BCS_BoneSpace
    translationMode=BMM_Additive
    translationSpace=BCS_BoneSpace
    alphaBoolBlend {blendOption=Linear}
    boneToModify {boneName=dyn_legArmor_l_master}
    rotation {pitch=10}
   }
   n16 = "Transform (Modify) Bone - Bone: dyn_frontArmor_master"(ComponentPose=@n15, Alpha=1, Translation="0,0,0", Rotation="-20,20,0", Scale="1,1,1") {
    rotationMode=BMM_Additive
    rotationSpace=BCS_BoneSpace
    translationMode=BMM_Additive
    translationSpace=BCS_BoneSpace
    alphaBoolBlend {blendOption=Linear}
    boneToModify {boneName=dyn_frontArmor_master}
    rotation {pitch=-20, yaw=20}
   }
   n17 = "Transform (Modify) Bone - Bone: dyn_backArmor_master"(ComponentPose=@n16, Alpha=1, Translation="0,0,0", Rotation="0,40,0", Scale="1,1,1") {
    rotationMode=BMM_Additive
    rotationSpace=BCS_BoneSpace
    translationMode=BMM_Additive
    translationSpace=BCS_BoneSpace
    alphaBoolBlend {blendOption=Linear}
    boneToModify {boneName=dyn_backArmor_master}
    rotation {yaw=40}
   }
   n18 = Anim Dynamics - Bone: dyn_chestPouch_a_01!(
    ComponentPose=Local To Component(LocalPose=Component To Local(ComponentPose=@n17))
    Alpha=1
    GravityOverride:Vector="0,0,150"
    ExternalForce="0,0,0"
   ) -> (Pose) {
    angularSpringConstant=50
    bAngularSpring=true
    bUseGravityOverride=true
    bUsePlanarLimit=false
    simulationSpace=BoneRelative
    alphaBoolBlend {blendOption=Linear}
    boundBone {boneName=dyn_chestPouch_a_01}
    gravityOverride {z=150}
    physicsBodyDefinitions [{
      boundBone {boneName=dyn_chestPouch_a_01}
      boxExtents=<1 1 1>
      constraintSetup {angularTargetAxis=AxisZ, angularLimitsMax=<1 4 3>, angularLimitsMin=<-1 0 -3>, angularTarget=<0 0 1>}
      localJointOffset=<0 0 0.1>
     }]
    relativeSpaceBone {boneName=spine_04}
   }
   n19 = Anim Dynamics - Bone: dyn_chestPouch_b_01!(ComponentPose=@n18, Alpha=1, GravityOverride:Vector="0,0,150", ExternalForce="0,0,0") -> (Pose) {
    angularSpringConstant=50
    bAngularSpring=true
    bUseGravityOverride=true
    bUsePlanarLimit=false
    simulationSpace=BoneRelative
    alphaBoolBlend {blendOption=Linear}
    boundBone {boneName=dyn_chestPouch_b_01}
    gravityOverride {z=150}
    physicsBodyDefinitions [{
      boundBone {boneName=dyn_chestPouch_b_01}
      boxExtents=<1 1 1>
      constraintSetup {angularTargetAxis=AxisZ, angularLimitsMax=<1 4 3>, angularLimitsMin=<-1 0 -3>, angularTarget=<0 0 1>}
      localJointOffset=<0 0 0.1>
     }]
    relativeSpaceBone {boneName=spine_04}
   }
   n20 = Anim Dynamics - Bone: dyn_chestPouch_c_01!(ComponentPose=@n19, Alpha=1, GravityOverride:Vector="0,0,150", ExternalForce="0,0,0") -> (Pose) {
    angularSpringConstant=50
    bAngularSpring=true
    bUseGravityOverride=true
    bUsePlanarLimit=false
    simulationSpace=BoneRelative
    alphaBoolBlend {blendOption=Linear}
    boundBone {boneName=dyn_chestPouch_c_01}
    gravityOverride {z=150}
    physicsBodyDefinitions [{
      boundBone {boneName=dyn_chestPouch_c_01}
      boxExtents=<1 1 1>
      constraintSetup {angularTargetAxis=AxisZ, angularLimitsMax=<1 4 3>, angularLimitsMin=<-1 0 -3>, angularTarget=<0 0 1>}
      localJointOffset=<0 0 0.1>
     }]
    relativeSpaceBone {boneName=spine_04}
   }
   n21 = Anim Dynamics - Bone: dyn_chestPouch_d_01!(ComponentPose=@n20, Alpha=1, GravityOverride:Vector="0,0,150", ExternalForce="0,0,0") -> (Pose) {
    angularSpringConstant=50
    bAngularSpring=true
    bUseGravityOverride=true
    bUsePlanarLimit=false
    simulationSpace=BoneRelative
    alphaBoolBlend {blendOption=Linear}
    boundBone {boneName=dyn_chestPouch_d_01}
    gravityOverride {z=150}
    physicsBodyDefinitions [{
      boundBone {boneName=dyn_chestPouch_d_01}
      boxExtents=<1 1 1>
      constraintSetup {angularTargetAxis=AxisZ, angularLimitsMax=<1 4 3>, angularLimitsMin=<-1 0 -3>, angularTarget=<0 0 1>}
      localJointOffset=<0 0 0.1>
     }]
    relativeSpaceBone {boneName=spine_04}
   }
   n22 = Anim Dynamics - Bone: dyn_chestPouch_e_01!(ComponentPose=@n21, Alpha=1, GravityOverride:Vector="0,0,150", ExternalForce="0,0,0") -> (Pose) {
    angularSpringConstant=50
    bAngularSpring=true
    bUseGravityOverride=true
    bUsePlanarLimit=false
    simulationSpace=BoneRelative
    alphaBoolBlend {blendOption=Linear}
    boundBone {boneName=dyn_chestPouch_e_01}
    gravityOverride {z=150}
    physicsBodyDefinitions [{
      boundBone {boneName=dyn_chestPouch_e_01}
      boxExtents=<1 1 1>
      constraintSetup {angularTargetAxis=AxisZ, angularLimitsMax=<1 4 3>, angularLimitsMin=<-1 0 -3>, angularTarget=<0 0 1>}
      localJointOffset=<0 0 0.1>
     }]
    relativeSpaceBone {boneName=spine_04}
   }
   n23 = Anim Dynamics - Bone: dyn_chestPouch_f_01!(ComponentPose=@n22, Alpha=1, GravityOverride:Vector="0,0,150", ExternalForce="0,0,0") -> (Pose) {
    angularSpringConstant=50
    bAngularSpring=true
    bUseGravityOverride=true
    bUsePlanarLimit=false
    simulationSpace=BoneRelative
    alphaBoolBlend {blendOption=Linear}
    boundBone {boneName=dyn_chestPouch_f_01}
    gravityOverride {z=150}
    physicsBodyDefinitions [{
      boundBone {boneName=dyn_chestPouch_f_01}
      boxExtents=<1 1 1>
      constraintSetup {angularTargetAxis=AxisZ, angularLimitsMax=<1 4 3>, angularLimitsMin=<-1 0 -3>, angularTarget=<0 0 1>}
      localJointOffset=<0 0 0.1>
     }]
    relativeSpaceBone {boneName=spine_04}
   }
   n24 = Anim Dynamics - Bone: dyn_chestPouch_g_01!(ComponentPose=@n23, Alpha=1, GravityOverride:Vector="0,0,150", ExternalForce="0,0,0") -> (Pose) {
    angularSpringConstant=50
    bAngularSpring=true
    bUseGravityOverride=true
    bUsePlanarLimit=false
    simulationSpace=BoneRelative
    alphaBoolBlend {blendOption=Linear}
    boundBone {boneName=dyn_chestPouch_g_01}
    gravityOverride {z=150}
    physicsBodyDefinitions [{
      boundBone {boneName=dyn_chestPouch_g_01}
      boxExtents=<1 1 1>
      constraintSetup {angularTargetAxis=AxisZ, angularLimitsMax=<1 4 3>, angularLimitsMin=<-1 0 -3>, angularTarget=<0 0 1>}
      localJointOffset=<0 0 0.1>
     }]
    relativeSpaceBone {boneName=spine_04}
   }
   n25 = "Transform (Modify) Bone - Bone: dyn_frontArmor_master"(ComponentPose=@n24, Alpha=1, Translation="0,0,0", Rotation="0,75,0", Scale="1,1,1") {
    rotationMode=BMM_Additive
    rotationSpace=BCS_BoneSpace
    translationMode=BMM_Additive
    translationSpace=BCS_BoneSpace
    alphaBoolBlend {blendOption=Linear}
    boneToModify {boneName=dyn_frontArmor_master}
    rotation {yaw=75}
   }
   n26 = "Transform (Modify) Bone - Bone: dyn_backArmor_master"(ComponentPose=@n25, Alpha=1, Translation="0,1,0", Rotation="0,10,0", Scale="1,1,1") {
    rotationMode=BMM_Additive
    rotationSpace=BCS_BoneSpace
    translationMode=BMM_Additive
    translationSpace=BCS_BoneSpace
    alphaBoolBlend {blendOption=Linear}
    boneToModify {boneName=dyn_backArmor_master}
    rotation {yaw=10}
    translation {y=1}
   }
   n27 = "Transform (Modify) Bone - Bone: dyn_backArmor_master"(ComponentPose=@n26, Alpha=1, Translation="0.25,0.5,0", Rotation="0,10,0", Scale="1,1,1") {
    rotationMode=BMM_Additive
    rotationSpace=BCS_BoneSpace
    translationMode=BMM_Additive
    translationSpace=BCS_BoneSpace
    alphaBoolBlend {blendOption=Linear}
    boneToModify {boneName=dyn_backArmor_master}
    rotation {yaw=10}
    translation=<0.25 0.5>
   }
   n28 = RigidBody!(ComponentPose=@n27, Alpha, OverrideWorldGravity:Vector="0,0,0", ExternalForce) -> (Pose) {
    bClampLinearTranslationLimitToRefPose=true
    bOverrideWorldGravity=true
    bTransferBoneVelocities=true
    overridePhysicsAsset=&/M_MED_SphericalDefense_Physics_Sides
    simulationSpace=BaseBoneSpace
    alphaBoolBlend {blendOption=Linear}
    baseBoneRef {boneName=pelvis}
    componentLinearAccScale=<0.2 0.2 0.2>
    componentLinearVelScale=<0.2 0.2 0.2>
    simSpaceSettings {worldAlpha=1}
   }
   n29 = RigidBody(ComponentPose=@n28, OverrideWorldGravity="0,0,0") {
    bClampLinearTranslationLimitToRefPose=true
    bOverrideWorldGravity=true
    overridePhysicsAsset=&/M_MED_SphericalDefense_Physics
    simulationSpace=BaseBoneSpace
    alphaBoolBlend {blendOption=Linear}
    baseBoneRef {boneName=pelvis}
    componentLinearAccScale=<0.4 0.4 0.4>
    componentLinearVelScale=<0.4 0.4 0.4>
    simSpaceSettings {worldAlpha=1}
   }
   n30 = RigidBody(ComponentPose=@n29, OverrideWorldGravity="0,0,0") {
    bClampLinearTranslationLimitToRefPose=true
    bOverrideWorldGravity=true
    overridePhysicsAsset=&/M_MED_SphericalDefense_FE_Physics
    simulationSpace=BaseBoneSpace
    alphaBoolBlend {blendOption=Linear}
    baseBoneRef {boneName=pelvis}
    componentLinearAccScale=<0.4 0.4 0.4>
    componentLinearVelScale=<0.4 0.4 0.4>
    simSpaceSettings {worldAlpha=1}
   }
   n31 = Anim Dynamics - Bone: dyn_legPouch_01!(ComponentPose=@n30, Alpha=1, GravityOverride:Vector="0,0,150", ExternalForce="0,0,0") -> (Pose) {
    angularSpringConstant=50
    bAngularSpring=true
    bUseGravityOverride=true
    bUsePlanarLimit=false
    simulationSpace=BoneRelative
    alphaBoolBlend {blendOption=Linear}
    boundBone {boneName=dyn_legPouch_01}
    gravityOverride {z=150}
    physicsBodyDefinitions [{
      boundBone {boneName=dyn_legPouch_01}
      boxExtents=<1 1 1>
      constraintSetup {angularTargetAxis=AxisZ, angularLimitsMax=<1 4 0>, angularLimitsMin=<-1 0 -6>, angularTarget=<0 0 1>}
      localJointOffset=<0 0 0.1>
     }]
    relativeSpaceBone {boneName=pelvis}
   }
   n32 = Anim Dynamics - Bone: dyn_tassel_l_01!(ComponentPose=@n31, Alpha=0.9, GravityOverride:Vector="0,0,150", ExternalForce="0,0,0") -> (Pose) {
    alpha=0.75
    angularDampingOverride=0.9999
    angularSpringConstant=50
    bAngularSpring=true
    bGravityOverrideInSimSpace=true
    bOverrideAngularDamping=true
    bUseGravityOverride=true
    simulationSpace=RootRelative
    alphaBoolBlend {blendOption=Linear}
    boundBone {boneName=dyn_tassel_l_01}
    componentLinearAccScale=<0.4 0.4 0.4>
    componentLinearVelScale=<0.4 0.4 0.4>
    gravityOverride {z=150}
    physicsBodyDefinitions [{
      boundBone {boneName=dyn_tassel_l_01}
      boxExtents=<15 15 15>
      constraintSetup {angularLimitsMax=<1 5 25>, angularLimitsMin=<-1 -35 -25>, angularTarget=<1 0 0>}
      localJointOffset=<4 0 0>
     }]
    relativeSpaceBone {boneName=root}
   }
   n33 = Anim Dynamics - Bone: dyn_tassel_l_02!(ComponentPose=@n32, Alpha=0.9, GravityOverride:Vector="0,0,150", ExternalForce="0,0,0") -> (Pose) {
    alpha=0.75
    angularDampingOverride=0.9999
    angularSpringConstant=50
    bAngularSpring=true
    bGravityOverrideInSimSpace=true
    bOverrideAngularDamping=true
    bUseGravityOverride=true
    simulationSpace=RootRelative
    alphaBoolBlend {blendOption=Linear}
    boundBone {boneName=dyn_tassel_l_02}
    componentLinearAccScale=<0.4 0.4 0.4>
    componentLinearVelScale=<0.4 0.4 0.4>
    gravityOverride {z=150}
    physicsBodyDefinitions [{
      boundBone {boneName=dyn_tassel_l_02}
      boxExtents=<19 19 19>
      constraintSetup {angularLimitsMax=<1 10 25>, angularLimitsMin=<-1 -55 -25>, angularTarget=<1 0 0>}
      localJointOffset=<4 0 -4>
     }]
    relativeSpaceBone {boneName=root}
   }
   n34 = Anim Dynamics - Bone: dyn_tassel_r_01!(ComponentPose=@n33, Alpha=0.9, GravityOverride:Vector="0,0,150", ExternalForce="0,0,0") -> (Pose) {
    alpha=0.75
    angularDampingOverride=0.9999
    angularSpringConstant=50
    bAngularSpring=true
    bGravityOverrideInSimSpace=true
    bOverrideAngularDamping=true
    bUseGravityOverride=true
    simulationSpace=RootRelative
    alphaBoolBlend {blendOption=Linear}
    boundBone {boneName=dyn_tassel_r_01}
    componentLinearAccScale=<0.4 0.4 0.4>
    componentLinearVelScale=<0.4 0.4 0.4>
    gravityOverride {z=150}
    physicsBodyDefinitions [{
      boundBone {boneName=dyn_tassel_r_01}
      boxExtents=<15 15 15>
      constraintSetup {angularLimitsMax=<1 5 25>, angularLimitsMin=<-1 -35 -25>, angularTarget=<1 0 0>}
      localJointOffset=<-4 0 0>
     }]
    relativeSpaceBone {boneName=root}
   }
   n35 = Anim Dynamics - Bone: dyn_tassel_r_02!(ComponentPose=@n34, Alpha=0.9, GravityOverride:Vector="0,0,150", ExternalForce="0,0,0") -> (Pose) {
    alpha=0.75
    angularDampingOverride=0.9999
    angularSpringConstant=50
    bAngularSpring=true
    bGravityOverrideInSimSpace=true
    bOverrideAngularDamping=true
    bUseGravityOverride=true
    simulationSpace=RootRelative
    alphaBoolBlend {blendOption=Linear}
    boundBone {boneName=dyn_tassel_r_02}
    componentLinearAccScale=<0.4 0.4 0.4>
    componentLinearVelScale=<0.4 0.4 0.4>
    gravityOverride {z=150}
    physicsBodyDefinitions [{
      boundBone {boneName=dyn_tassel_r_02}
      boxExtents=<19 19 19>
      constraintSetup {angularLimitsMax=<1 10 25>, angularLimitsMin=<-1 -55 -25>, angularTarget=<1 0 0>}
      localJointOffset=<-4 0 4>
     }]
    relativeSpaceBone {boneName=root}
   }
   n36 = Anim Dynamics - Bone: dyn_footStrap_out_01!(ComponentPose=@n35, Alpha=0.5, GravityOverride:Vector="0,0,150", ExternalForce="0,0,0") -> (Pose) {
    angularDampingOverride=0.999
    angularSpringConstant=200
    bAngularSpring=true
    bOverrideAngularDamping=true
    bUseGravityOverride=true
    bUsePlanarLimit=false
    simulationSpace=RootRelative
    alphaBoolBlend {blendOption=Linear}
    boundBone {boneName=dyn_footStrap_out_01}
    componentAppliedLinearAccClamp=<6000 6000 6000>
    componentLinearAccScale=<0.3 0.3 0.4>
    componentLinearVelScale=<0.3 0.3 0.4>
    gravityOverride {z=150}
    physicsBodyDefinitions [{
      boundBone {boneName=dyn_footStrap_out_01}
      boxExtents=<15 15 15>
      constraintSetup {angularConstraintType=Cone, angularLimitsMax=<5 15 15>, angularLimitsMin=<-5 -15 -15>, angularTarget=<1 0 0>}
      localJointOffset=<2 0 0>
     }]
    relativeSpaceBone {boneName=def_knot_01}
   }
   n37 = Anim Dynamics - Bone: dyn_footStrap_in_01!(ComponentPose=@n36, Alpha=0.5, GravityOverride:Vector="0,0,150", ExternalForce="0,0,0") -> (Pose) {
    angularDampingOverride=0.999
    angularSpringConstant=200
    bAngularSpring=true
    bOverrideAngularDamping=true
    bUseGravityOverride=true
    bUsePlanarLimit=false
    simulationSpace=RootRelative
    alphaBoolBlend {blendOption=Linear}
    boundBone {boneName=dyn_footStrap_in_01}
    componentAppliedLinearAccClamp=<6000 6000 6000>
    componentLinearAccScale=<0.3 0.3 0.4>
    componentLinearVelScale=<0.3 0.3 0.4>
    gravityOverride {z=150}
    physicsBodyDefinitions [{
      boundBone {boneName=dyn_footStrap_in_01}
      boxExtents=<15 15 15>
      constraintSetup {angularConstraintType=Cone, angularLimitsMax=<5 15 15>, angularLimitsMin=<-5 -15 -15>, angularTarget=<1 0 0>}
      localJointOffset=<2 0 0>
     }]
    relativeSpaceBone {boneName=def_knot_01}
   }
   Output Pose(Result=Component To Local(ComponentPose=@n37)) {layerGroup=None}
  ]}
 EventGraph {nodes [
   Event BlueprintUpdateAnimation()
   > n1 = Sequence!(execute) -> (then_0, then_1, then_2:Exec, then_3:Exec, then_4:Exec)
   > n2 = Gravity Override State Function DEPRECATED(
    CharacterPartAnimInstance=Self-Reference()
    GameTimeInSeconds=GetGameTimeInSeconds()
    GravityOverrideParams=Get Gravity Override Params()
    RBAN Params=Get RBAN Params()
    DebugDraw=Get doDebug() -> (doDebug:bool)
   )
   > Set GravityOverride(GravityOverride=@n2)
   > n3 = Set PawnSpeed(PawnSpeed:double=Vector Length(A=Get PawnVelocity())) -> (Output_Get:double)
   > Branch(Condition=Get doDebug())
   > PrintString!(execute, InString=Append(A=@n2.CurrentStateOut, B="To String (Float)"(InDouble=@n3)), bPrintToScreen, bPrintToLog, TextColor, Duration, Key) -> (then)
   Set Is DBNO Float(execute=@n1.then_1, Is DBNO Float:double="To Float (Boolean)"(InBool=Get bIsDBNO())) -> (Output_Get:double)
   > Set Is Baseball Sliding(Is Baseball Sliding:double="To Float (Boolean)"(InBool=Get bIsBaseballSliding())) -> (Output_Get:double)
   > Set Is Tact Sprinting Alpha(Is Tact Sprinting Alpha:double="To Float (Boolean)"(InBool=Get bIsTacticalSprinting())) -> (Output_Get:double)
   Set Is Skydiving Float(execute=@n1.then_2, Is Skydiving Float:double="To Float (Boolean)"(InBool=Get bIsSkyDiving())) -> (Output_Get:double)
   > Set Is Knee Sliding Alpha(Is Knee Sliding Alpha:double="To Float (Boolean)"(InBool=Get bIsKneeSliding())) -> (Output_Get:double)
   > Set Is Playing Emote Alpha(Is Playing Emote Alpha:double="To Float (Boolean)"(InBool=Get bIsPlayingEmote())) -> (Output_Get:double)
   n4 = Get bIsPlayingEmote()
   > Set Is Not Playing Emote Alpha(Is Not Playing Emote Alpha:double="To Float (Boolean)"(InBool=NOT Boolean(A=@n4))) -> (Output_Get:double)
   n5 = Cast To FortPlayerAnimInstance(execute=@n1.then_3, Object=GetAnimInstance(self=GetOwnerSkeletalMeshComponent()))
   > Set IsPistol(IsPistol:bool=true) -> (Output_Get:bool)
   > n6 = Switch on EFortWeaponCoreAnimation!(execute, Selection:EFortWeaponCoreAnimation Enum=Get WeaponCoreAnim(self=@n5)) -> (
    Melee:Exec
    Pistol:Exec
    Shotgun:Exec
    PaperBlueprint:Exec
    Rifle:Exec
    MeleeOneHand:Exec
    MachinePistol:Exec
    RocketLauncher:Exec
    GrenadeLauncher:Exec
    GoingCommando:Exec
    AssaultRifle:Exec
    TacticalShotgun:Exec
    SniperRifle:Exec
    TrapPlacement:Exec
    ShoulderLauncher:Exec
    AbilityDecoTool:Exec
    Crossbow:Exec
    C4:Exec
    RemoteControl:Exec
    DualWield:Exec
    AR_BullPup:Exec
    AR_ForwardGrip:Exec
    MedPackPaddles:Exec
    SMG_P90:Exec
    AR_DrumGun:Exec
    Consumable_Small:Exec
    Consumable_Large:Exec
    Balloon:Exec
    MountedTurret:Exec
    CreativeTool:Exec
    ExplosiveBow:Exec
    AshtonIndigo:Exec
    AshtonChicago:Exec
    MeleeDualWield:Exec
    Unarmed:Exec
   )
   Set IsPistol(execute=@n6.Pistol, IsPistol=true)
   Set Is Pistol Alpha(execute=@n1.then_4, Is Pistol Alpha:double="To Float (Boolean)"(InBool=Get IsPistol() -> (IsPistol:bool))) -> (Output_Get:double)
  ]}
]
variables [
 doDebug {type=boolean}
 Gravity Override Params {
  type=GravityOverrideParamsStruct
  default {
   onGround_Standing {gravityOverride=<0 0 -3000>}
   skydive_DivingDown {gravityOverride=<-3000 0 0>}
   skydive_FloatingDown {gravityOverride=<0 0 3500>}
   skydive_GoingUp {gravityOverride=<0 -300 2200>}
   skydive_Surfing {gravityOverride=<500 -300 980>}
  }
 }
 GravityOverride {type=Vector}
 Is Baseball Sliding {type=number}
 Is DBNO Float {type=number}
 Is Knee Sliding Alpha {type=number}
 Is Not Playing Emote Alpha {type=number}
 Is Pistol Alpha {type=number}
 Is Playing Emote Alpha {type=number}
 Is Skydiving Float {type=number}
 Is Tact Sprinting Alpha {type=number}
 IsPistol {type=boolean}
 PawnSpeed {type=number}
 RBAN Params {
  type=RBANParamsStruct
  default {
   dBNO {simSpaceSettings {
     maxAngularAcceleration=1000
     maxAngularVelocity=3
     maxLinearAcceleration=300
     maxLinearVelocity=700
     velocityScaleZ=0
     externalLinearDragV=<0 0.5 0>
    }}
   dBNO_Carried {simSpaceSettings {
     maxAngularAcceleration=1000
     maxAngularVelocity=3
     maxLinearAcceleration=300
     maxLinearVelocity=700
     velocityScaleZ=0
     externalLinearDragV=<0 0.5 0>
    }}
   emoteOrMelee {simSpaceSettings {
     maxAngularAcceleration=1000
     maxAngularVelocity=3
     maxLinearAcceleration=300
     maxLinearVelocity=700
     velocityScaleZ=0
     externalLinearDragV=<0 0.5 0>
    }}
   front_End {simSpaceSettings {
     maxAngularAcceleration=1000
     maxAngularVelocity=3
     maxLinearAcceleration=300
     maxLinearVelocity=700
     velocityScaleZ=0
     externalLinearDragV=<0 0.5 0>
    }}
   onGround_Crouch_Moving {simSpaceSettings {
     maxAngularAcceleration=1000
     maxAngularVelocity=3
     maxLinearAcceleration=300
     maxLinearVelocity=700
     velocityScaleZ=0
     externalLinearDragV=<0 0.5 0>
    }}
   onGround_Falling {simSpaceSettings {
     maxAngularAcceleration=1000
     maxAngularVelocity=3
     maxLinearAcceleration=300
     maxLinearVelocity=700
     velocityScaleZ=0
     externalLinearDragV=<0 0.5 0>
    }}
   onGround_Moving {simSpaceSettings {
     maxAngularAcceleration=1000
     maxAngularVelocity=3
     maxLinearAcceleration=300
     maxLinearVelocity=700
     velocityScaleZ=0
     externalLinearDragV=<0 0.5 0>
    }}
   onGround_Standing {simSpaceSettings {
     maxAngularAcceleration=1000
     maxAngularVelocity=3
     maxLinearAcceleration=300
     maxLinearVelocity=700
     velocityScaleZ=0
     externalLinearDragV=<0 0.5 0>
    }}
   rail_Grinding {simSpaceSettings {
     maxAngularAcceleration=1000
     maxAngularVelocity=3
     maxLinearAcceleration=300
     maxLinearVelocity=700
     velocityScaleZ=0
     worldAlpha=0.1
     externalLinearDragV=<0 0.5 0>
    }}
   riding_Moving {simSpaceSettings {
     maxAngularAcceleration=1000
     maxAngularVelocity=3
     maxLinearAcceleration=300
     maxLinearVelocity=700
     velocityScaleZ=0
     externalLinearDragV=<0 0.5 0>
    }}
   riding_Moving_Targeting {simSpaceSettings {
     maxAngularAcceleration=1000
     maxAngularVelocity=3
     maxLinearAcceleration=300
     maxLinearVelocity=700
     velocityScaleZ=0
     externalLinearDragV=<0 0.5 0>
    }}
   riding_Standing {simSpaceSettings {
     maxAngularAcceleration=1000
     maxAngularVelocity=3
     maxLinearAcceleration=300
     maxLinearVelocity=700
     velocityScaleZ=0
     externalLinearDragV=<0 0.5 0>
    }}
   skydive_DivingDown {simSpaceSettings {
     maxAngularAcceleration=1000
     maxAngularVelocity=3
     maxLinearAcceleration=300
     maxLinearVelocity=700
     velocityScaleZ=0
     externalLinearDragV=<0 0.5 0>
    }}
   skydive_FloatingDown {
    componentLinearAccScale=<0 0.5 0>
    componentLinearVelScale=<0 0.5 0>
    simSpaceSettings {
     maxAngularAcceleration=1000
     maxAngularVelocity=3
     maxLinearAcceleration=300
     maxLinearVelocity=700
     velocityScaleZ=0
     externalLinearDragV=<0 0.5 0>
    }
   }
   skydive_GoingUp {simSpaceSettings {
     maxAngularAcceleration=1000
     maxAngularVelocity=3
     maxLinearAcceleration=300
     maxLinearVelocity=700
     velocityScaleZ=0
     externalLinearDragV=<0 0.5 0>
    }}
   skydive_Parachuting {
    componentLinearAccScale=<0 0.1 0.3>
    componentLinearVelScale=<0 0.1 0.3>
    simSpaceSettings {
     maxAngularAcceleration=1000
     maxAngularVelocity=3
     maxLinearAcceleration=300
     maxLinearVelocity=700
     velocityScaleZ=0
     externalLinearDragV=<0 0.5 0>
    }
   }
   skydive_Surfing {simSpaceSettings {
     maxAngularAcceleration=1000
     maxAngularVelocity=3
     maxLinearAcceleration=300
     maxLinearVelocity=700
     velocityScaleZ=0
     externalLinearDragV=<0 0.5 0>
    }}
   swimming {simSpaceSettings {
     maxAngularAcceleration=1000
     maxAngularVelocity=3
     maxLinearAcceleration=300
     maxLinearVelocity=700
     velocityScaleZ=0
     externalLinearDragV=<0 0.5 0>
    }}
   tactical_Sprint {simSpaceSettings {
     maxAngularAcceleration=1000
     maxAngularVelocity=3
     maxLinearAcceleration=300
     maxLinearVelocity=700
     velocityScaleZ=0
     externalLinearDragV=<0 0.5 0>
    }}
   ziplining {simSpaceSettings {
     maxAngularAcceleration=1000
     maxAngularVelocity=3
     maxLinearAcceleration=300
     maxLinearVelocity=700
     velocityScaleZ=0
     externalLinearDragV=<0 0.5 0>
    }}
  }
 }
]
